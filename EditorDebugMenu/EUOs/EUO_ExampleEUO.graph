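# EUO_ExampleEUO is an example Editor Utility Blueprint in EditorDebugMenu that launches player pawns upward by teleporting them.
# Its Run event gets the game world through UnrealEditorSubsystem and checks it is valid, so it acts on the game world from editor tooling.
# It finds every FortPlayerPawn, prints the count, then teleports each 2500 units up in Z, keeping its rotation.
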EditorUtilityBlueprint /EditorDebugMenu/EUOs/EUO_ExampleEUO
graphs [EventGraph {nodes [
   Event Run()
   > n1 = GetGameWorld!(execute, self:Unreal Editor Subsystem Object Reference=Get UnrealEditorSubsystem()) -> (then, ReturnValue:World Object Reference)
   > Is Valid(InputObject=@n1)
   > n2 = GetAllActorsOfClass!(execute, WorldContextObject:Object Reference=@n1, ActorClass:Actor Class Reference=/Script/FortniteGame.FortPlayerPawn) -> (then, OutActors:Array of Fort Player Pawn Object References)
   > PrintText(InText=Format Text!(
     Format="LOCTEXT(\"[EUO Example] ------------ There are {num} actors in the level -------------\")"
     num:int=Length(TargetArray:Array of Fort Player Pawn Object References=@n2)
    ) -> (Result))
   > n3 = For Each Loop(Array:Array of Fort Player Pawn Object References=@n2) -> (Array Element:Fort Player Pawn Object Reference)
   n4 = Get Actor Transform!(self=@n3) -> (ReturnValue_Location:Vector, ReturnValue_Rotation:Rotator, ReturnValue_Scale:Vector)
   > Teleport!(
    execute
    self:Actor Object Reference=@n3
    DestLocation:Vector=vector + vector(A=@n4, B="0,0,2500")
    DestRotation:Rotator=@n4.ReturnValue_Rotation
   ) -> (then, ReturnValue:bool)
  ]}]
variables []
